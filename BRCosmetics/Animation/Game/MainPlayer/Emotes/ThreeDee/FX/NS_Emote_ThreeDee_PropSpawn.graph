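# Niagara system for the ThreeDee emote's prop spawn, layering effects on the gadget prop and the owner skeleton.
# A delayed Beam emitter samples both skeletons to stretch and orient meshes between them, adding swirl, sparks and light.
# Ember and light-shaft bursts align to that axis; ring and speed-line emitters follow two propeller sockets.
# Spawn groups separate layers within emitters, and time-of-day, platform and front-end inputs adjust brightness.
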NiagaraSystem /BRCosmetics/Animation/Game/MainPlayer/Emotes/ThreeDee/FX/NS_Emote_ThreeDee_PropSpawn
dependencies {
 usedDataInterfaces [
  &NiagaraDataInterfaceCurve
  &NiagaraDataInterfacePlatformSet
  &NiagaraDataInterfaceColorCurve
  &NiagaraDataInterfaceVectorCurve
  &NiagaraDataInterfaceVector2DCurve
  &NiagaraDataInterfaceSocketReader
  &NiagaraDataInterfaceArrayInt32
 ]
 usedDynamicInputs [&SimulationPosition]
 usedModules [&SystemState]
}
emitters {
 Beam {
  emitterUpdateScript [
   &EmitterState(
    Inactive Response=::NewEnumerator1
    Life Cycle Mode=::NewEnumerator1
    Loop Behavior=::NewEnumerator1
    Loop Delay=2.26
    Loop Duration=0.7
    UseLoopDelay=true
   )
   &SpawnBurst_Instantaneous(Spawn Time=0.07, Use Spawn Probability=false, UseLoopCountLimit=false)
   &SpawnBurst_Instantaneous(Spawn Group=1, Spawn Time=0.05, Use Spawn Probability=false, UseLoopCountLimit=false) {module=SpawnBurst_Instantaneous001}
   &SpawnBurst_Instantaneous(Spawn Group=2, Spawn Time=0.02, Use Spawn Probability=false, UseLoopCountLimit=false) {module=SpawnBurst_Instantaneous002}
   &SpawnBurst_Instantaneous(Spawn Group=3, Use Spawn Probability=false, UseLoopCountLimit=false) {module=SpawnBurst_Instantaneous003}
   &SpawnBurst_Instantaneous(Spawn Group=4, Spawn Time=0.02, Use Spawn Probability=false, UseLoopCountLimit=false) {module=SpawnBurst_Instantaneous004}
   &SpawnBurst_Instantaneous(Spawn Count=5, Spawn Group=5, Use Spawn Probability=false, UseLoopCountLimit=false) {module=SpawnBurst_Instantaneous005}
   /SetVariables(Emitter.SplineWidth_Base:NiagaraFloat=&FloatFromCurve(
     FloatCurve=DataInterface() {propertyValues {Curve {
        defaultValue=2
        keys [
         {arriveTangent=4.2973, interpMode=RCIM_Cubic, leaveTangent=4.2973, tangentMode=RCTM_User, time=0, value=0.4}
         {interpMode=RCIM_Cubic, tangentMode=RCTM_User, time=0.299, value=1}
         {
          arriveTangentWeight=0.1919
          interpMode=RCIM_Cubic
          leaveTangent=-19.4564
          leaveTangentWeight=0.2123
          tangentMode=RCTM_Break
          time=0.38
          value=1
         }
         {arriveTangent=-1.2961, interpMode=RCIM_Cubic, leaveTangent=-1.2961, tangentMode=RCTM_User, time=0.52, value=0}
        ]
       }}}
     CurveIndex=$Emitter.NormalizedLoopAge
     Scale Curve=23
    )) {isSetParameters=true}
   /SetVariables(Emitter.SplineWidth_Tip:NiagaraFloat=&FloatFromCurve(
     FloatCurve=DataInterface() {propertyValues {Curve {keys [
         <at 0 1>
         {
          arriveTangentWeight=0.2036
          interpMode=RCIM_Cubic
          leaveTangent=-20.2413
          leaveTangentWeight=0.0679
          tangentMode=RCTM_Break
          time=0.42
          value=1
         }
         {arriveTangent=-1.7225, arriveTangentWeight=0.9348, interpMode=RCIM_Cubic, leaveTangent=-1.7225, tangentMode=RCTM_User, time=0.52, value=0}
        ]}}}
     CurveIndex=$Emitter.NormalizedLoopAge
     Scale Curve=9
    )) {isSetParameters=true}
   /SetVariables(Emitter.SplineDarkWidth_Base:NiagaraFloat=&FloatFromCurve(
     FloatCurve=DataInterface() {propertyValues {Curve {
        defaultValue=2
        keys [
         {arriveTangent=0.1726, interpMode=RCIM_Cubic, leaveTangent=0.1726, tangentMode=RCTM_User, time=0.0238, value=0}
         {interpMode=RCIM_Cubic, tangentMode=RCTM_User, time=0.1684, value=1}
         {
          arriveTangentWeight=1.0145
          interpMode=RCIM_Cubic
          leaveTangent=-12.438
          leaveTangentWeight=0.6231
          tangentMode=RCTM_Break
          time=0.38
          value=1
         }
         <at 0.51 0 RCIM_Cubic>
        ]
       }}}
     CurveIndex=$Emitter.NormalizedLoopAge
     Scale Curve=25
    )) {isSetParameters=true}
   /SetVariables(Emitter.SplineDarkWidth_Tip:NiagaraFloat=&FloatFromCurve(
     FloatCurve=DataInterface() {propertyValues {Curve {
        defaultValue=1
        keys [
         {tangentMode=RCTM_User, time=0, value=1}
         {
          arriveTangent=-16.7162
          arriveTangentWeight=0.1234
          interpMode=RCIM_Cubic
          leaveTangent=-16.7162
          leaveTangentWeight=0.0267
          tangentMode=RCTM_User
          time=0.41
          value=1
         }
         {arriveTangent=-4.9601, interpMode=RCIM_Cubic, leaveTangent=-4.9601, tangentMode=RCTM_User, time=0.51, value=0}
        ]
       }}}
     CurveIndex=$Emitter.NormalizedLoopAge
     Scale Curve=7
    )) {isSetParameters=true}
   /SetVariables(Emitter.SplineCenterWidth_Base:NiagaraFloat=&FloatFromCurve(
     FloatCurve=DataInterface() {propertyValues {Curve {
        defaultValue=2
        keys [
         {arriveTangent=0.5786, interpMode=RCIM_Cubic, leaveTangent=0.5786, tangentMode=RCTM_User, time=0, value=0.6614}
         {interpMode=RCIM_Cubic, tangentMode=RCTM_User, time=0.3189, value=1}
         {
          arriveTangentWeight=0.049
          interpMode=RCIM_Cubic
          leaveTangent=-21.6175
          leaveTangentWeight=0.0272
          tangentMode=RCTM_Break
          time=0.37
          value=1
         }
         {arriveTangent=-1.2493, interpMode=RCIM_Cubic, leaveTangent=-1.2493, tangentMode=RCTM_User, time=0.52, value=0}
        ]
       }}}
     CurveIndex=$Emitter.NormalizedLoopAge
     Scale Curve=20
    )) {isSetParameters=true}
   /SetVariables(Emitter.SplineCenterWidth_Tip:NiagaraFloat=&FloatFromCurve(
     FloatCurve=DataInterface() {propertyValues {Curve {keys [
         <at 0 1>
         {arriveTangent=-16.9732, interpMode=RCIM_Cubic, leaveTangent=-16.9732, tangentMode=RCTM_User, time=0.41, value=1}
         {arriveTangent=-2.0115, interpMode=RCIM_Cubic, leaveTangent=-2.0115, tangentMode=RCTM_User, time=0.52, value=0}
        ]}}}
     CurveIndex=$Emitter.NormalizedLoopAge
     Scale Curve=4
    )) {isSetParameters=true}
   /SetVariables(Emitter.SplineWidthRed_Base:NiagaraFloat=&FloatFromCurve(
     FloatCurve=DataInterface() {propertyValues {Curve {
        defaultValue=2
        keys [
         {arriveTangent=4.2973, interpMode=RCIM_Cubic, leaveTangent=4.2973, tangentMode=RCTM_User, time=0, value=0.4}
         {interpMode=RCIM_Cubic, tangentMode=RCTM_User, time=0.2982, value=1}
         {arriveTangentWeight=0.1096, interpMode=RCIM_Cubic, leaveTangentWeight=0.2339, tangentMode=RCTM_User, time=0.38, value=1}
         {arriveTangent=-4.1181, interpMode=RCIM_Cubic, leaveTangent=-4.1181, tangentMode=RCTM_User, time=0.52, value=0}
        ]
       }}}
     CurveIndex=$Emitter.NormalizedLoopAge
     Scale Curve=35
    )) {isSetParameters=true}
   /SetVariables(Emitter.SplineWidthRed_Tip:NiagaraFloat=&FloatFromCurve(
     FloatCurve=DataInterface() {propertyValues {Curve {keys [
         <at 0 1>
         {arriveTangent=-4.6776, interpMode=RCIM_Cubic, leaveTangent=-4.6776, tangentMode=RCTM_User, time=0.41, value=1}
         {arriveTangent=-3.4862, interpMode=RCIM_Cubic, leaveTangent=-3.4862, tangentMode=RCTM_User, time=0.52, value=0}
        ]}}}
     CurveIndex=$Emitter.NormalizedLoopAge
     Scale Curve=12
    )) {isSetParameters=true}
  ]
  particleSpawnScript [
   /SetVariables(Particles.VisibilityTag=&SpawnGroup()) {isSetParameters=true}
   &InitializeParticle(Color=<rgba 0.6661 0.1497 0 1.5>, Lifetime=0.5, Mesh Scale Mode=::NewEnumerator3, UsePositionOffset=false, Write Material Random=true)
   &InitialMeshOrientation(Mesh Orientation Mode=::NewEnumerator2, Orientation Vector=<0 0 1>, Rotation=<0 0.5 0>, Use Rotation Vector=false)
   /SetVariables(Particles.Sparks Lifetime=0.35, Particles.Swirl Lifetime=0.45) {isSetParameters=true}
  ]
  particleUpdateScript [
   &ParticleState()
   /SetVariables(
    Particles.Update Sparks Age:NiagaraFloat=&Divide_Float(A=$Particles.Age, B=$Particles.Initial.Sparks Lifetime)
    Particles.Update Swirl Age:NiagaraFloat=&Divide_Float(A=$Particles.Age, B=$Particles.Initial.Swirl Lifetime)
   ) {isSetParameters=true}
   &SkeletalMeshLocation(
    Bone Sampling Mode=::NewEnumerator4
    MaskSpawnGroup=false
    Orientation Sampling=::NewEnumerator1
    Skeletal Mesh=DataInterface() {propertyValues {
      DefaultMesh=null
      ExcludeBoneName=root
      OptionalComponentClass=null
      PreviewMesh=/BRCosmetics/Animation/Props/Gadgets_ThreeDee/Meshes/ThreeDee_Gadget
      SkinningMode=SkinOnTheFly
      SoftSourceActor=null
      SourceMode=Default
      UvSetIndex=0
      WholeMeshLOD=-1
      bExcludeBone=false
      bReadDeformedGeometry=true
      bRequireCurrentFrameData=true
      ComponentTags []
      FilteredBones [offset_anim_body_JNT]
      FilteredSockets []
      MeshUserParameter {parameter {name=None, typeDefHandle {registeredTypeIndex=1}, varData []}}
      SamplingRegions []
     }}
   )
   &SkeletalMeshLocation(
    Bone Sampling Mode=::NewEnumerator4
    MaskSpawnGroup=false
    Orientation Sampling=::NewEnumerator1
    Position Sampling=::NewEnumerator2
    Sampled Position Offset=<-5 0 0>
    Skeletal Mesh=$User.OwnerSkeleton
   ) {module=SkeletalMeshLocation001}
   &ShapeLocation(
    Additional Quaternion Rotation=false
    Cylinder Height=60
    Cylinder Height Midpoint=0
    Cylinder Radius=7
    MaskSpawnGroup=true
    Offset=&Multiply_VectorByFloat(
     Vector=&Subtract_Vector(
      A=Linked() {linkedVariable {name=Particles.SkeletalMeshLocation001.SampledPosition, type=NiagaraPosition}}
      B=Linked() {linkedVariable {name=Particles.SkeletalMeshLocation.SampledPosition, type=NiagaraPosition}}
     )
     Float=&FloatFromCurve(
      FloatCurve=DataInterface() {propertyValues {Curve {
         defaultValue=1
         keys [{arriveTangent=1.2172, interpMode=RCIM_Cubic, leaveTangent=1.2172, tangentMode=RCTM_User, time=0, value=0}, <at 1 0.4307 RCIM_Cubic>]
        }}}
      CurveIndex=$Particles.Update Sparks Age
      Scale Curve=&RandomRangeFloat(Minimum=0.5, Maximum=1)
     )
    )
    Offset Mode=::NewEnumerator0
    Override Seed=false
    Rotation Mode=::NewEnumerator3
    Rotation Quaternion=&OrientMeshToVector(
     Look At Direction=&Subtract_Position(A=$Particles.SkeletalMeshLocation.SampledPosition, B=$Particles.SkeletalMeshLocation001.SampledPosition)
     Reference Axis Vector=<1 0 0>
    )
    Shape Origin=$Particles.SkeletalMeshLocation.SampledPosition
    Shape Primitive=::NewEnumerator1
    Spawn Group Mask=5
    Surface Only=false
    UseOwnerScale=true
   )
   &ScaleMeshSize(Scale Factor=&Multiply_Vector(
     A=&MakeVector(
      X=1
      Y=1
      Z=&Multiply_Float(
       A=&FloatFromCurve(FloatCurve=DataInterface() {propertyValues {Curve {keys [
            {arriveTangent=10.1233, interpMode=RCIM_Cubic, leaveTangent=10.1233, tangentMode=RCTM_User, time=0, value=0.3506}
            {arriveTangent=4.1037, interpMode=RCIM_Cubic, leaveTangent=4.1037, time=0.0211, value=0.8088}
            {arriveTangentWeight=0.0969, interpMode=RCIM_Cubic, leaveTangentWeight=0.3745, tangentMode=RCTM_User, time=0.1582, value=1}
           ]}}})
       B=&RemapRange(
        Input Value To Remap=&VectorLength(Vector=&Subtract_Position(A=$Particles.SkeletalMeshLocation.SampledPosition, B=$Particles.SkeletalMeshLocation001.SampledPosition))
        Input Min=0
        Input Max=150
        Output Min=0
        Output Max=2.5
       )
      )
     )
     B=&Divide_Vector(A:Vector3f=<1 1 1>, B:Vector3f=$Engine.Owner.Scale)
    ))
   &ScaleColor(
    Scale Alpha=&Multiply_Float(
     A=&FloatFromCurve(
      FloatCurve=DataInterface() {propertyValues {Curve {keys [
          <at 0 1>
          {arriveTangent=-0.2775, interpMode=RCIM_Cubic, leaveTangent=-0.2775, tangentMode=RCTM_User, time=0.8, value=1}
          {arriveTangent=-2.6012, interpMode=RCIM_Cubic, leaveTangent=-2.6012, tangentMode=RCTM_User, time=0.9, value=0.85}
          {arriveTangent=-14.0132, interpMode=RCIM_Cubic, leaveTangent=-14.0132, tangentMode=RCTM_User, time=1, value=0}
         ]}}}
      Scale Curve=&FNTimeOfDayScalar(Night Scale=0.6, Morning Scale=0.8, Evening Scale=0.6)
     )
     B=&FN_FrontEndFloatModifier(Gameplay=1, Front End=0.7)
    )
    ScaleA=true
    ScaleRGB=true
   )
   /SetVariables(Particles.color002:LinearColor=&MakeLinearColorFromVectorAndFloat(
     "Vector (RGB)"=&MakeVectorFromLinearColorRGB(Color=<rgba 1 1 1 1>)
     "Float (Alpha)"=&Multiply_Float(
      A=&FloatFromCurve(
       FloatCurve=DataInterface() {propertyValues {Curve {keys [<at 0 1>,<at 0.95 1>]}}}
       Scale Curve=&FNTimeOfDayScalar(Night Scale=0.2, Morning Scale=0.2, Day Scale=0.3, Evening Scale=0.2)
      )
      B=&Multiply_Float(
       A=&FN_FrontEndFloatModifier(Gameplay=1, Front End=0.8)
       B=&FN_PickFloatDependOnPlatform(TargetPlatformActive=1, TargetPlatformNoActive=0.8)
      )
     )
    )) {isSetParameters=true}
   /SetVariables(Particles.Color003:LinearColor=&MakeLinearColorFromVectorAndFloat(
     "Vector (RGB)"=&MakeVectorFromLinearColorRGB(Color=<rgba 1.5 1.5 1.5 1>)
     "Float (Alpha)"=&Multiply_Float(
      A=&FloatFromCurve(
       FloatCurve=DataInterface() {propertyValues {Curve {keys [<at 0 5>,<at 0.1 1>,<at 0.9 1>]}}}
       Scale Curve=&FNTimeOfDayScalar(Night Scale=0.6, Morning Scale=0.8, Evening Scale=0.6)
      )
      B=&FN_PickFloatDependOnPlatform(TargetPlatformActive=1, TargetPlatformNoActive=0.15)
     )
    )) {isSetParameters=true}
   /SetVariables(Particles.Color004:LinearColor=&MakeLinearColorFromVectorAndFloat(
     "Vector (RGB)"=&MakeVectorFromLinearColorRGB(Color=<rgba 1 1 1 1.5>)
     "Float (Alpha)"=&Multiply_Float(
      A=&FloatFromCurve(
       FloatCurve=DataInterface() {propertyValues {Curve {keys [<at 0 1>,<at 0.85 1>]}}}
       Scale Curve=&FNTimeOfDayScalar(Night Scale=0.6, Morning Scale=0.8, Evening Scale=0.6)
      )
      B=&FN_PickFloatDependOnPlatform(TargetPlatformActive=1, TargetPlatformNoActive=0.6)
     )
    )) {isSetParameters=true}
   &DynamicMaterialParameters(
    Index 0 Param 1=&FloatFromCurve(
     FloatCurve=DataInterface() {propertyValues {Curve {keys [
         <at -0.0006 1 RCIM_Cubic>
         {arriveTangentWeight=0.0517, interpMode=RCIM_Cubic, leaveTangentWeight=0.373, tangentMode=RCTM_User, time=0.1, value=-1}
         {
          arriveTangentWeight=6.2214
          interpMode=RCIM_Cubic
          leaveTangent=17.8388
          leaveTangentWeight=1.6616
          tangentMode=RCTM_Break
          time=0.51
          value=-1
         }
         {arriveTangent=0.7491, interpMode=RCIM_Cubic, leaveTangent=0.7491, tangentMode=RCTM_User, time=0.7134, value=1}
        ]}}}
     Scale Curve=&RandomRangeFloat(Minimum=0.9, Maximum=1)
    )
    Index 0 Param 2=0.4
    Index 0 Param 3=1
    Index 2 Param 1=0.1
    Index 2 Param 3=0.5
    Module.Index0Float0Write=true
    Module.Index0Float1Write=true
    Module.Index0Float2Write=true
    Module.Index0Float3Write=true
    Module.Index2Float0Write=false
    Module.Index2Float1Write=false
    Module.Index2Float2Write=true
    Module.Index2Float3Write=false
    Write Parameter Index 2=::NewEnumerator1
   )
   /SetVariables(
    Particles.SwirlColor:LinearColor=&ColorFromCurve(
     ColorCurve=DataInterface() {propertyValues {
       AlphaCurve {keys [<at 0.6949 1 RCIM_Cubic>,<at 1 0 RCIM_Cubic>]}
       BlueCurve {keys [<at 0 1.0662>, {arriveTangent=-1.0662, interpMode=RCIM_Cubic, leaveTangent=-1.0662, time=0.5811, value=0}, <at 1 0 RCIM_Cubic>]}
       GreenCurve {keys [<at 0 3.8452>, {arriveTangent=-3.6315, interpMode=RCIM_Cubic, leaveTangent=-3.6315, time=0.5811, value=2.8945}, <at 1 0.2138 RCIM_Cubic>]}
       RedCurve {keys [<at 0 5>, {arriveTangent=-4, interpMode=RCIM_Cubic, leaveTangent=-4, time=0.5811, value=5}, <at 1 1 RCIM_Cubic>]}
      }}
     CurveIndex=$Particles.Update Swirl Age
    )
    Particles.SwirlDynamicParameters:Vector4f=&MakeVector4(
     X=&FloatFromCurve(FloatCurve=DataInterface() {propertyValues {Curve {keys [<at 0 1>,<at 1 0>]}}})
     Y=0
     Z=0
     W=0
    )
    Particles.SwirlPosition:NiagaraPosition=&AddVectorToPosition(
     Position=$Particles.Position
     Vector=&VectorFromCurve(
      VectorCurve=DataInterface() {propertyValues {
        XCurve {keys [<at 0 1>,<at 1 1>]}
        YCurve {keys [<at 0 1>,<at 1 1>]}
        ZCurve {keys [{arriveTangent=-2.2967, interpMode=RCIM_Cubic, leaveTangent=-2.2967, tangentMode=RCTM_User, time=0, value=-0.1694}, <at 1 -1 RCIM_Cubic>]}
       }}
      CurveIndex=$Particles.Update Swirl Age
      Scale Curve=<1 1 30>
     )
    )
    Particles.SwirlSize:Vector3f=&VectorFromCurve(
     VectorCurve=DataInterface() {propertyValues {
       XCurve {keys [
         {arriveTangent=12.6282, interpMode=RCIM_Cubic, leaveTangent=12.6282, tangentMode=RCTM_User, time=-0.0068, value=0.1207}
         {arriveTangentWeight=0.1738, interpMode=RCIM_Cubic, leaveTangentWeight=0.1273, tangentMode=RCTM_User, time=0.1126, value=0.9}
         {arriveTangent=-0.2665, interpMode=RCIM_Cubic, leaveTangent=-0.2665, tangentMode=RCTM_User, time=0.2, value=0.9}
         {arriveTangent=-1.125, interpMode=RCIM_Cubic, leaveTangent=-1.125, time=0.411, value=0.4056}
         {arriveTangent=-2.5384, interpMode=RCIM_Cubic, leaveTangent=-2.5384, tangentMode=RCTM_User, time=1, value=0}
        ]}
       YCurve {keys [
         {arriveTangent=12.6282, interpMode=RCIM_Cubic, leaveTangent=12.6282, tangentMode=RCTM_User, time=-0.0068, value=0.1207}
         {arriveTangentWeight=0.1784, interpMode=RCIM_Cubic, leaveTangentWeight=0.1227, tangentMode=RCTM_User, time=0.1157, value=0.9}
         {arriveTangent=-0.2665, interpMode=RCIM_Cubic, leaveTangent=-0.2665, tangentMode=RCTM_User, time=0.2, value=0.9}
         {arriveTangent=-1.125, interpMode=RCIM_Cubic, leaveTangent=-1.125, time=0.413, value=0.4186}
         {arriveTangent=-2.5384, interpMode=RCIM_Cubic, leaveTangent=-2.5384, tangentMode=RCTM_User, time=1, value=0}
        ]}
       ZCurve {keys [
         {arriveTangent=4.7271, interpMode=RCIM_Cubic, leaveTangent=4.7271, tangentMode=RCTM_User, time=0, value=0.1407}
         {arriveTangent=1.0476, interpMode=RCIM_Cubic, leaveTangent=1.0476, time=0.0664, value=0.7863}
         <at 0.8202 1 RCIM_Cubic>
        ]}
      }}
     CurveIndex=$Particles.Update Swirl Age
     Scale Curve=<0.25 0.25 0.8>
    )
   ) {isSetParameters=true}
   &UpdateMeshOrientation(
    Rotation Rate=&MaskFloatBySpawnGroup(Masked Float=-7, Passthrough Float=0, Spawn Group Mask=4)
    Rotation Vector=<0 0 1>
   )
   &UpdateMeshOrientation(
    Facing Direction=&Subtract_Position(A=$Particles.SkeletalMeshLocation.SampledPosition, B=$Particles.SkeletalMeshLocation001.SampledPosition)
    Facing Interpolation=false
    Orientation Method=::NewEnumerator1
    Override Quat=false
    TransformFacing=false
    TransformUp=false
   ) {module=UpdateMeshOrientation001}
   /SetVariables(Particles.MeshOrientation=&MultiplyQuaternion(
     Quaternion A=$Particles.MeshOrientation
     Quaternion B=&OrientMeshToVector(
      Look At Direction=&Subtract_Position(A=$Particles.SkeletalMeshLocation.SampledPosition, B=$Particles.SkeletalMeshLocation001.SampledPosition)
      Reference Axis Vector=<1 0 0>
     )
    )) {enabled=false, isSetParameters=true}
   &Light_Attributes(
    Color=1
    Light Position Offset:Vector3f=&Multiply_VectorByFloat(
     Vector=&Subtract_Vector(
      A=Linked() {linkedVariable {name=Particles.SkeletalMeshLocation001.SampledPosition, type=NiagaraPosition}}
      B=Linked() {linkedVariable {name=Particles.SkeletalMeshLocation.SampledPosition, type=NiagaraPosition}}
     )
     Float=0.5
    )
    Light Radius=130
    Position=1
    Position Offset Coordinate Space:ENiagaraCoordinateSpace=::Local
    Radius=1
    Scale Alpha:NiagaraFloat=&FloatFromCurve(FloatCurve=DataInterface() {propertyValues {Curve {
        defaultValue=1
        keys [
         {tangentMode=RCTM_User, time=0, value=1}
         {arriveTangent=-0.0405, interpMode=RCIM_Cubic, leaveTangent=-0.0405, tangentMode=RCTM_User, time=0.7009, value=1}
         {arriveTangent=-8.0089, interpMode=RCIM_Cubic, leaveTangent=-8.0089, tangentMode=RCTM_User, time=1, value=0}
        ]
       }}})
    Scale RGB:Vector3f=&MakeVectorFromLinearColorRGB(Color=<rgba 7 6.8637 1.4486 1>)
    ScaleA=true
    ScaleRGB=true
    UsePositionOffset=true
   )
   /SetVariables(
    Particles.Sparks Alignment:Vector3f=&Subtract_Position(A=$Particles.SkeletalMeshLocation.SampledPosition, B=$Particles.SkeletalMeshLocation001.SampledPosition)
    Particles.Sparks Color:LinearColor=&ColorFromCurve(
     ColorCurve=DataInterface() {propertyValues {
       AlphaCurve {keys [<at 0.5885 1 RCIM_Cubic>,<at 1 0 RCIM_Cubic>]}
       BlueCurve {keys [<at 0 0 RCIM_Cubic>,<at 1 0 RCIM_Cubic>]}
       GreenCurve {keys [<at 0 0.9276 RCIM_Cubic>,<at 1 0.6205 RCIM_Cubic>]}
       RedCurve {keys [<at 0 1 RCIM_Cubic>,<at 1 1 RCIM_Cubic>]}
      }}
     CurveIndex=$Particles.Update Sparks Age
    )
    Particles.Sparks Size:Vector2f=&Vector2DFromCurve(
     Vector2Curve=DataInterface() {propertyValues {
       XCurve {keys [<at 0 0.0777>,<at 0.3513 1>,<at 0.801 1>,<at 1 0>]}
       YCurve {keys [
         {arriveTangent=2.6034, interpMode=RCIM_Cubic, leaveTangent=2.6034, tangentMode=RCTM_User, time=0, value=0.6795}
         {arriveTangentWeight=0.1056, interpMode=RCIM_Cubic, leaveTangentWeight=0.3117, tangentMode=RCTM_User, time=0.3747, value=2}
         {arriveTangentWeight=0.4793, interpMode=RCIM_Cubic, leaveTangentWeight=0.2377, tangentMode=RCTM_User, time=0.7086, value=2}
         <at 0.971 0 RCIM_Cubic>
        ]}
      }}
     CurveIndex=$Particles.Update Sparks Age
     ScaleCurve=&MakeVector2D(X=1, Y=&RandomRangeFloat(Minimum=10, Maximum=20))
    )
   ) {isSetParameters=true}
  ]
  rendererClasses [
   &NiagaraMeshRendererProperties
   &NiagaraLightRendererProperties
   &NiagaraSpriteRendererProperties
  ]
  renderers [
   {rendererClass=&NiagaraMeshRendererProperties, rendererIndex=0}
   {rendererClass=&NiagaraMeshRendererProperties, rendererIndex=1}
   {rendererClass=&NiagaraMeshRendererProperties, rendererIndex=2}
   {rendererClass=&NiagaraMeshRendererProperties, rendererIndex=3}
   {rendererClass=&NiagaraMeshRendererProperties, rendererIndex=4}
   {rendererClass=&NiagaraLightRendererProperties, rendererIndex=5}
   {rendererClass=&NiagaraSpriteRendererProperties, rendererIndex=6}
   {rendererClass=&NiagaraSpriteRendererProperties, rendererIndex=7}
  ]
 }
 Embers_Burst {
  emitterUpdateScript [
   &EmitterState(
    Inactive Response=::NewEnumerator1
    Life Cycle Mode=::NewEnumerator1
    Loop Behavior=::NewEnumerator1
    Loop Delay=2.63
    Loop Duration=1
    UseLoopDelay=true
   )
   &SpawnBurst_Instantaneous(Spawn Count=5, Spawn Group=1, Use Spawn Probability=false, UseLoopCountLimit=false) {module=SpawnBurst_Instantaneous001}
   &SpawnBurst_Instantaneous(Spawn Count=15, Use Spawn Probability=false, UseLoopCountLimit=false)
  ]
  particleSpawnScript [
   /SetVariables(Particles.VisibilityTag=&SpawnGroup()) {isSetParameters=true}
   &InitializeParticle(
    Color=<rgba 1 1 1 2>
    Lifetime=&UniformRangedFloat(Minimum=0.6, Maximum=0.4)
    Mass Mode=::NewEnumerator1
    Sprite Size Mode=::NewEnumerator3
    Uniform Sprite Size=&MaskFloatBySpawnGroup(
     Masked Float=&RandomRangeFloat(Minimum=5, Maximum=10)
     Passthrough Float=&RandomRangeFloat(Minimum=0.8, Maximum=1.5)
    )
    UsePositionOffset=false
    Write Material Random=true
   ) {module=InitializeParticle001}
   &SkeletalMeshLocation(
    Bone Sampling Mode=::NewEnumerator4
    MaskSpawnGroup=false
    Orientation Sampling=::NewEnumerator1
    Skeletal Mesh=DataInterface() {propertyValues {
      DefaultMesh=null
      ExcludeBoneName=root
      OptionalComponentClass=null
      PreviewMesh=&ThreeDee_Gadget
      SkinningMode=SkinOnTheFly
      SoftSourceActor=null
      SourceMode=Default
      UvSetIndex=0
      WholeMeshLOD=-1
      bExcludeBone=false
      bReadDeformedGeometry=true
      bRequireCurrentFrameData=true
      ComponentTags []
      FilteredBones [offset_anim_body_JNT]
      FilteredSockets []
      MeshUserParameter {parameter {name=None, typeDefHandle {registeredTypeIndex=1}, varData []}}
      SamplingRegions []
     }}
   )
   &SkeletalMeshLocation(
    Bone Sampling Mode=::NewEnumerator4
    MaskSpawnGroup=false
    Orientation Sampling=::NewEnumerator1
    Position Sampling=::NewEnumerator2
    Sampled Position Offset=<-5 0 0>
    Skeletal Mesh=$User.OwnerSkeleton
   ) {module=SkeletalMeshLocation001}
   &ShapeLocation(
    Additional Quaternion Rotation=false
    Box Midpoint=<0.5 1 1>
    Box Size=<1 0.1 0.1>
    MaskSpawnGroup=false
    Non Uniform Scale=<70 1 1>
    Offset=<0 0 -50>
    Offset Coordinate Space=::World
    Offset Mode=::NewEnumerator0
    Override Seed=false
    Rotation Mode=::NewEnumerator3
    Rotation Quaternion=&OrientMeshToVector(
     Look At Direction=&Subtract_Position(A=$Particles.SkeletalMeshLocation.SampledPosition, B=$Particles.SkeletalMeshLocation001.SampledPosition)
     Reference Axis Vector=<1 0 0>
    )
    Shape Primitive=::NewEnumerator2
    Surface Only=false
    UseOwnerScale=true
   )
   &AddVelocity(Velocity=&RandomRangeVector(Minimum=<20 20 0>, Maximum=<-20 -20 0>)) {module=AddVelocity001}
   &CameraOffset(Camera Offset Amount=-10)
  ]
  particleUpdateScript [
   &ParticleState()
   &ApplyOwnerScaleToAttributes(Scale Initial Velocity=true)
   &Drag(
    Drag=&Multiply_Float(
     A=&FloatFromCurve(
      FloatCurve=DataInterface() {propertyValues {Curve {keys [<at 0 0.5>, {arriveTangent=0.5, interpMode=RCIM_Cubic, leaveTangent=0.5, time=0.3, value=0.5}, <at 1 1 RCIM_Cubic>]}}}
      Scale Curve=12
     )
     B=1
    )
    Use Linear Drag=true
    Use Rotational Drag=false
   )
   /Niagara/Modules/Update/Forces/CurlNoiseForce(
    Noise Frequency=10
    Noise Strength=&FloatFromCurve(FloatCurve=DataInterface() {propertyValues {Curve {keys [<at 0 1>,<at 1 0>]}}}, Scale Curve=2000)
    Pan Noise=true
    Pan Noise Field=<0.1 0.1 0>
   )
   &GravityForce(Gravity=&RandomRangeVector(Minimum=<0 0 900>, Maximum=<0 0 -500>))
   &InheritSourceMovement(
    Applied Position Scale=&MakeVector(
     X=0
     Y=0
     Z=&FloatFromCurve(
      FloatCurve=DataInterface() {propertyValues {Curve {
         defaultValue=1
         keys [
          {interpMode=RCIM_Cubic, tangentMode=RCTM_User, time=0, value=0.0874}
          {arriveTangent=3.8536, interpMode=RCIM_Cubic, leaveTangent=3.8536, tangentMode=RCTM_User, time=0.1, value=0.5931}
         ]
        }}}
      Scale Curve=-1
     )
    )
    ApplyForce=false
    ApplyPosition=true
    ApplyPositionFalloff=false
   )
   &SolveForcesAndVelocity(Clamp Velocity=false, Limit Acceleration=false)
   &ScaleSpriteSizeBySpeed(
    Max Scale Factor=<0.8 2>
    Min Scale Factor=<0.8 0.8>
    Sample Scale Factor By Curve=true
    Scale Factor Curve=DataInterface() {propertyValues {Curve {keys [
        <at -0.0077 0.0765 RCIM_Cubic>
        {arriveTangent=1.2463, interpMode=RCIM_Cubic, leaveTangent=1.2463, tangentMode=RCTM_User, time=0.9985, value=0.9549}
       ]}}}
    Velocity Threshold=300
   )
   &ScaleColor(
    Scale Alpha=&FloatFromCurve(
     FloatCurve=DataInterface() {propertyValues {Curve {keys [<at 0.25 1>,<at 1 0>]}}}
     Scale Curve=&MaskFloatBySpawnGroup(
      Masked Float=&FNTimeOfDayScalar(Night Scale=0.1, Morning Scale=0.2, Day Scale=0.3, Evening Scale=0.2)
      Passthrough Float=&FNTimeOfDayScalar(Night Scale=0.1, Morning Scale=0.2, Evening Scale=0.2)
     )
    )
    Scale RGB=&MakeVectorFromLinearColorRGB(Color=&ColorFromCurve(ColorCurve=DataInterface() {propertyValues {
        AlphaCurve {keys [<at 0 1 RCIM_Cubic>,<at 0.7107 1>]}
        BlueCurve {keys [<at 0 0.09 RCIM_Cubic>,<at 1 0.1753 RCIM_Cubic>]}
        GreenCurve {keys [<at 0 1.2384 RCIM_Cubic>,<at 1 0.3243 RCIM_Cubic>]}
        RedCurve {keys [<at 0 3 RCIM_Cubic>,<at 1 1 RCIM_Cubic>]}
       }}))
    ScaleA=true
    ScaleRGB=true
   )
  ]
  rendererClasses [&NiagaraSpriteRendererProperties]
  renderers [
   {rendererClass=&NiagaraSpriteRendererProperties}
   {rendererClass=&NiagaraSpriteRendererProperties, rendererIndex=1}
  ]
 }
 LightShaftsBurst {
  emitterUpdateScript [
   &EmitterState(
    Inactive Response=::NewEnumerator1
    Life Cycle Mode=::NewEnumerator1
    Loop Behavior=::NewEnumerator1
    Loop Delay=2.26
    Loop Duration=1
    UseLoopDelay=true
   )
   &SpawnBurst_Instantaneous(Use Spawn Probability=false, UseLoopCountLimit=false)
  ]
  particleSpawnScript [
   &InitializeParticle(
    Color=<rgba 2 0.9604 0 1>
    Lifetime=0.7
    Mesh Scale Mode=::NewEnumerator3
    Mesh Uniform Scale=0.05
    Position Offset=<0 0 -5>
    UsePositionOffset=true
    Write Material Random=true
   )
   &InitialMeshOrientation(Orientation Vector=<0 0 -1>, Use Orientation Vector=true, Use Rotation Vector=false)
  ]
  particleUpdateScript [
   &ParticleState()
   &SkeletalMeshLocation(
    Bone Sampling Mode=::NewEnumerator4
    MaskSpawnGroup=false
    Orientation Sampling=::NewEnumerator1
    Skeletal Mesh=DataInterface() {propertyValues {
      DefaultMesh=null
      ExcludeBoneName=root
      OptionalComponentClass=null
      PreviewMesh=&ThreeDee_Gadget
      SkinningMode=SkinOnTheFly
      SoftSourceActor=null
      SourceMode=Default
      UvSetIndex=0
      WholeMeshLOD=-1
      bExcludeBone=false
      bReadDeformedGeometry=true
      bRequireCurrentFrameData=true
      ComponentTags []
      FilteredBones [offset_anim_body_JNT]
      FilteredSockets []
      MeshUserParameter {parameter {name=None, typeDefHandle {registeredTypeIndex=1}, varData []}}
      SamplingRegions []
     }}
   )
   &SkeletalMeshLocation(
    Bone Sampling Mode=::NewEnumerator4
    MaskSpawnGroup=false
    Orientation Sampling=::NewEnumerator1
    Position Sampling=::NewEnumerator2
    Sampled Position Offset=<-5 0 0>
    Skeletal Mesh=$User.OwnerSkeleton
   ) {module=SkeletalMeshLocation001}
   /SetVariables(Particles.MeshOrientation=&OrientMeshToVector(
     Look At Direction=&Subtract_Position(A=$Particles.SkeletalMeshLocation.SampledPosition, B=$Particles.SkeletalMeshLocation001.SampledPosition)
     Reference Axis Vector=<1 0 0>
    )) {isSetParameters=true}
   &ScaleMeshSize(Scale Factor=&Multiply_VectorByFloat(
     Vector=<1 1 1>
     Float=&FloatFromCurve(FloatCurve=DataInterface() {propertyValues {Curve {
         defaultValue=1
         keys [{interpMode=RCIM_Cubic, tangentMode=RCTM_User, time=0, value=1}, {interpMode=RCIM_Cubic, tangentMode=RCTM_User, time=0.4038, value=0}]
        }}})
    ))
   &DynamicMaterialParameters(
    Index 0 Param 1=&FloatFromCurve(
     FloatCurve=DataInterface() {propertyValues {Curve {keys [
         {arriveTangent=-1.392, interpMode=RCIM_Cubic, leaveTangent=-1.392, tangentMode=RCTM_User, time=0, value=1}
         {arriveTangent=0.0302, interpMode=RCIM_Cubic, leaveTangent=0.0302, tangentMode=RCTM_User, time=1, value=0.3676}
        ]}}}
     Scale Curve=0.2071
    )
    Index 0 Param 2=&Lerp_Float(
     A=0.05
     B=0.1
     Alpha=&FloatFromCurve(FloatCurve=DataInterface() {propertyValues {Curve {keys [
          {arriveTangent=4.7763, interpMode=RCIM_Cubic, leaveTangent=4.7763, tangentMode=RCTM_User, time=0, value=0}
          {arriveTangent=0.4097, interpMode=RCIM_Cubic, leaveTangent=0.4097, tangentMode=RCTM_User, time=0.3, value=0.9}
          <at 1 1>
         ]}}})
    )
    Index 0 Param 3=&FloatFromCurve(
     FloatCurve=DataInterface() {propertyValues {Curve {keys [{arriveTangent=1.5696, interpMode=RCIM_Cubic, leaveTangent=1.5696, tangentMode=RCTM_User, time=0, value=0.1857}, <at 0.4 0.5854 RCIM_Cubic>]}}}
     CurveIndex=$Particles.Age
     Scale Curve=0.02
    )
    Index 0 Param 4=3
    Module.Index0Float0Write=true
    Module.Index0Float1Write=true
    Module.Index0Float2Write=true
    Module.Index0Float3Write=true
   )
   &ScaleColor(
    Scale RGB=&VectorFromFloat(Value=&FloatFromCurve(
      FloatCurve=DataInterface() {propertyValues {Curve {keys [
          {arriveTangent=9.0725, interpMode=RCIM_Cubic, leaveTangent=9.0725, tangentMode=RCTM_User, time=0, value=0}
          {arriveTangent=-2.5249, interpMode=RCIM_Cubic, leaveTangent=-2.5249, tangentMode=RCTM_User, time=0.1408, value=0.9847}
          {arriveTangent=-0.0405, interpMode=RCIM_Cubic, leaveTangent=-0.0405, tangentMode=RCTM_User, time=1, value=0}
         ]}}}
      CurveIndex=$Particles.Age
      Scale Curve=5
     ))
    ScaleA=true
    ScaleRGB=true
   ) {module=ScaleColor001}
  ]
  rendererClasses [&NiagaraMeshRendererProperties]
  renderers [{rendererClass=&NiagaraMeshRendererProperties, rendererIndex=0}]
 }
 Main_RingCombined {
  emitterUpdateScript [
   &EmitterState(Life Cycle Mode=::NewEnumerator1, Loop Behavior=::NewEnumerator1, Loop Delay=0.958, Loop Duration=8.089, UseLoopDelay=true)
   &SpawnRate(Spawn Probability=0.6, SpawnRate=10, Use Spawn Probability=true)
  ]
  particleSpawnScript [
   &PartitionParticles()
   /SetVariables(Particles.RandomRange:NiagaraFloat=&Multiply_Float(
     A=&RandomRangeFloat(Minimum=0.25, Maximum=0.375)
     B=&FloatFromCurve(
      FloatCurve=DataInterface() {propertyValues {Curve {keys [<at 0 1 RCIM_Cubic>, {arriveTangent=7.3942, interpMode=RCIM_Cubic, leaveTangent=7.3942, tangentMode=RCTM_User, time=0.3, value=2}]}}}
      CurveIndex=$Emitter.Age
     )
    )) {isSetParameters=true}
   &InitializeParticle(
    Color=&FNTimeOfDayColor(
     Night Color=<rgba 0.8 0.9367 1 0.6>
     Morning Color=<rgba 0.5958 0.6976 0.7448 0.6>
     Day Color=<rgba 0.8 0.9367 1 0.6>
     Evening Color=<rgba 0.575 0.6732 0.7188 0.6>
    )
    Lifetime Max=0.2
    Lifetime Min=0.4
    Lifetime Mode=::NewEnumerator1
    Mesh Scale=&Multiply_VectorByFloat(Vector=<1 2 1>, Float=0.06)
    Mesh Scale Mode=::NewEnumerator4
    UsePositionOffset=false
    Write Material Random=true
   )
   &CameraOffset(Camera Offset Amount=15)
   &InitialMeshOrientation(
    Mesh Axis To Orient=<0 1 0>
    Mesh Orientation Mode=::NewEnumerator2
    Orientation Vector=<0 0 1>
    Rotation=&RandomRangeVector(Minimum=<0 0 0>, Maximum=<0 1 0>)
    Use Rotation Vector=true
   )
  ]
  particleUpdateScript [
   &ParticleState()
   &SocketLocation(
    Orientation Sampling=::NewEnumerator1
    SamplePosOffset=true
    Sampled Position Offset=&MakeVector(
     X=0
     Y=0
     Z=&FloatFromCurve(FloatCurve=DataInterface() {propertyValues {Curve {
         defaultValue=1
         keys [
          {arriveTangent=-40.6612, interpMode=RCIM_Cubic, leaveTangent=-40.6612, tangentMode=RCTM_User, time=0, value=2.3}
          {interpMode=RCIM_Cubic, tangentMode=RCTM_User, time=1, value=-15}
         ]
        }}})
    )
    Socket Index=&SelectIntFromArray(
     Direct Array Index=$Particles.PartitionParticles.Partition
     Direct Array Mode=::NewEnumerator1
     Int Selection Array=DataInterface() {propertyValues {IntData [0,1]}}
    )
    Socket Reader=DataInterface() {propertyValues {
      AttachComponentClass=null
      AttachComponentTag=None
      EditorPreviewAsset=&ThreeDee_Gadget
      SourceActorRef=null
      SourceAsset=null
      SourceMode=Default
      bRequireCurrentFrameData=true
      bUpdateSocketsPerFrame=true
      FilteredSockets [propellorLF1_JNT, propellorRF1_JNT]
      ObjectParameterBinding {parameter {name=None, typeDefHandle {registeredTypeIndex=1}, varData []}}
     }}
    Socket Sampling Mode=::NewEnumerator3
    VelocitySpeedLimit=false
   )
   &DynamicMaterialParameters(
    Index 0 Param 1=&FloatFromCurve(FloatCurve=DataInterface() {propertyValues {Curve {keys [
         {arriveTangent=2.3202, interpMode=RCIM_Cubic, leaveTangent=2.3202, tangentMode=RCTM_User, time=0, value=0.2481}
         {arriveTangent=0.6377, interpMode=RCIM_Cubic, leaveTangent=0.6377, time=0.0995, value=0.4178}
         {interpMode=RCIM_Cubic, tangentMode=RCTM_User, time=0.3951, value=0.5}
         {arriveTangent=-1.1781, tangentMode=RCTM_User, time=1, value=0}
        ]}}})
    Module.Index0Float0Write=true
    Module.Index0Float1Write=true
    Module.Index0Float2Write=true
    Module.Index0Float3Write=true
   )
   &ScaleColor(
    Scale Alpha=&FloatFromCurve(
     FloatCurve=DataInterface() {propertyValues {Curve {keys [
         {arriveTangent=-0.0871, interpMode=RCIM_Cubic, leaveTangent=-0.0871, tangentMode=RCTM_User, time=0, value=0.6231}
         {arriveTangent=-2.1635, interpMode=RCIM_Cubic, leaveTangent=-2.1635, tangentMode=RCTM_User, time=0.247, value=0.5413}
         {tangentMode=RCTM_User, time=1, value=0}
        ]}}}
     Scale Curve=&FNTimeOfDayScalar(Night Scale=0.2, Morning Scale=0.3, Day Scale=0.5, Evening Scale=0.3)
    )
    Scale RGB=&MakeVectorFromLinearColorRGB(Color=&ColorFromCurve(ColorCurve=DataInterface() {propertyValues {
        AlphaCurve {keys []}
        BlueCurve {keys [
          <at 0 1 RCIM_Cubic>
          {arriveTangent=-1.2161, interpMode=RCIM_Cubic, leaveTangent=-1.2161, time=0.1313, value=1}
          <at 0.4231 0.4855>
          <at 1 0.2309>
         ]}
        GreenCurve {keys [
          <at 0 1 RCIM_Cubic>
          {arriveTangent=-1.2767, interpMode=RCIM_Cubic, leaveTangent=-1.2767, time=0.1313, value=0.8855}
          <at 0.4231 0.4599>
          <at 1 0.2157>
         ]}
        RedCurve {keys [
          <at 0 1 RCIM_Cubic>
          {arriveTangent=-1.3057, interpMode=RCIM_Cubic, leaveTangent=-1.3057, time=0.1313, value=0.8293}
          <at 0.4231 0.4476>
          <at 1 0.2075>
         ]}
       }}))
    ScaleA=true
    ScaleRGB=true
   )
   &UpdateMeshOrientation(Rotation Rate=0.2, Rotation Vector=<0 1 0>) {module=UpdateMeshOrientation001}
  ]
  rendererClasses [&NiagaraMeshRendererProperties]
  renderers [{rendererClass=&NiagaraMeshRendererProperties, rendererIndex=0}]
 }
 NE_SpeedLinesCombined001 {
  emitterUpdateScript [
   &EmitterState(Life Cycle Mode=::NewEnumerator1, Loop Behavior=::NewEnumerator1, Loop Delay=0.958, Loop Duration=8.189, UseLoopDelay=true)
   &SpawnRate(SpawnRate=10, Use Spawn Probability=false)
  ]
  particleSpawnScript [
   &PartitionParticles()
   &InitializeParticle(
    Color=<rgba 1 1 1 5>
    Lifetime=0.24
    Mesh Scale=&Multiply_VectorByFloat(Vector=<1 1 3>, Float=1)
    Mesh Scale Mode=::NewEnumerator4
    UsePositionOffset=false
    Write Material Random=true
   ) {module=InitializeParticle001}
  ]
  particleUpdateScript [
   &ParticleState()
   &SocketLocation(
    Orientation Sampling=::NewEnumerator1
    SamplePosOffset=true
    Socket Index=&SelectIntFromArray(
     Direct Array Index=$Particles.PartitionParticles.Partition
     Direct Array Mode=::NewEnumerator1
     Int Selection Array=DataInterface() {propertyValues {IntData [0,1]}}
    )
    Socket Reader=DataInterface() {propertyValues {
      AttachComponentClass=null
      AttachComponentTag=None
      EditorPreviewAsset=&ThreeDee_Gadget
      SourceActorRef=null
      SourceAsset=null
      SourceMode=Default
      bRequireCurrentFrameData=true
      bUpdateSocketsPerFrame=true
      FilteredSockets [propellorLF1_JNT, propellorRF1_JNT]
      ObjectParameterBinding {parameter {name=None, typeDefHandle {registeredTypeIndex=1}, varData []}}
     }}
    Socket Sampling Mode=::NewEnumerator3
    VelocitySpeedLimit=false
   )
   &SolveForcesAndVelocity(Clamp Velocity=false, Limit Acceleration=false)
   &DynamicMaterialParameters(
    Index 0 Param 1=1
    Index 0 Param 2=&RandomRangeFloat(Minimum=1.5, Maximum=2)
    Index 0 Param 3=1
    Index 0 Param 4=0.1
    Module.Index0Float0Write=true
    Module.Index0Float1Write=true
    Module.Index0Float2Write=true
    Module.Index0Float3Write=true
   )
   &ScaleColor(
    Scale Alpha=&FloatFromCurve(
     FloatCurve=DataInterface() {propertyValues {Curve {keys [
         {arriveTangent=44.6077, interpMode=RCIM_Cubic, leaveTangent=44.6077, tangentMode=RCTM_User, time=0, value=0}
         <at 0.0781 1.0138 RCIM_Cubic>
         {arriveTangent=-1.8607, interpMode=RCIM_Cubic, leaveTangent=-1.8607, tangentMode=RCTM_User, time=1, value=0}
        ]}}}
     Scale Curve=&FNTimeOfDayScalar(Night Scale=0.1, Morning Scale=0.2, Day Scale=0.2, Evening Scale=0.2)
    )
    ScaleA=true
    ScaleRGB=true
   )
   &ScaleMeshSize(Scale Factor=&MakeVector(
     X=1
     Y=1
     Z=&FloatFromCurve(
      FloatCurve=DataInterface() {propertyValues {Curve {keys [{arriveTangent=2.9993, interpMode=RCIM_Cubic, leaveTangent=2.9993, tangentMode=RCTM_User, time=0, value=0}, <at 1 1 RCIM_Cubic>]}}}
      Scale Curve=0.8
     )
    ))
  ]
  rendererClasses [&NiagaraMeshRendererProperties]
  renderers [{rendererClass=&NiagaraMeshRendererProperties, rendererIndex=0}]
 }
 Wind {
  emitterUpdateScript [
   &EmitterState(
    Life Cycle Mode=::NewEnumerator1
    Loop Behavior=::NewEnumerator1
    Loop Duration Mode=::NewEnumerator1
    Scalability Mode=::NewEnumerator1
    UseLoopDelay=false
   )
   &SpawnRate(
    SpawnRate=&FloatFromCurve(
     FloatCurve=DataInterface() {propertyValues {Curve {
        defaultValue=1
        keys [
         {arriveTangent=-0.0405, interpMode=RCIM_Cubic, leaveTangent=-0.0405, tangentMode=RCTM_User, time=0, value=1.5}
         {arriveTangent=-4.5342, interpMode=RCIM_Cubic, leaveTangent=-0.2568, tangentMode=RCTM_Break, time=0.2, value=1}
         {arriveTangentWeight=0.5096, interpMode=RCIM_Cubic, leaveTangentWeight=0.0004, tangentMode=RCTM_User, time=1.357, value=1}
         {arriveTangentWeight=0.001, interpMode=RCIM_Cubic, leaveTangentWeight=0.3517, tangentMode=RCTM_User, time=1.358, value=0}
         {arriveTangentWeight=2.3342, interpMode=RCIM_Cubic, leaveTangentWeight=0.0003, tangentMode=RCTM_User, time=8.289, value=0}
         <at 8.29 1 RCIM_Cubic>
        ]
       }}}
     CurveIndex=$Emitter.Age
     Scale Curve=15
    )
    Use Spawn Probability=false
   )
   &SpawnRate(
    SpawnGroup=1
    SpawnRate=&FloatFromCurve(
     FloatCurve=DataInterface() {propertyValues {Curve {
        defaultValue=1
        keys [
         {arriveTangent=-0.0405, interpMode=RCIM_Cubic, leaveTangent=-0.0405, tangentMode=RCTM_User, time=0, value=1.5}
         {arriveTangent=-4.5342, interpMode=RCIM_Cubic, leaveTangent=-0.2568, tangentMode=RCTM_Break, time=0.2, value=1}
         {arriveTangentWeight=0.5096, interpMode=RCIM_Cubic, leaveTangentWeight=0.0004, tangentMode=RCTM_User, time=1.357, value=1}
         {arriveTangentWeight=0.001, interpMode=RCIM_Cubic, leaveTangentWeight=0.3517, tangentMode=RCTM_User, time=1.358, value=0}
         {arriveTangentWeight=2.3342, interpMode=RCIM_Cubic, leaveTangentWeight=0.0003, tangentMode=RCTM_User, time=8.289, value=0}
         <at 8.29 1 RCIM_Cubic>
        ]
       }}}
     CurveIndex=$Emitter.Age
     Scale Curve=15
    )
    Use Spawn Probability=false
   ) {module=SpawnRate001}
  ]
  particleSpawnScript [
   /SetVariables(Particles.RibbonID=/Niagara/DynamicInputs/Meshes/MakeRibbonIDFromSpawnGroup()) {isSetParameters=true}
   &InitializeParticle(
    Lifetime=0.3
    Ribbon Width=&RandomRangeFloat(Minimum=10, Maximum=20)
    Ribbon Width Mode=::NewEnumerator1
    Sprite Size Mode=::NewEnumerator3
    UsePositionOffset=false
    Write Material Random=true
   )
   &SocketLocation(
    Orientation Sampling=::NewEnumerator1
    SamplePosOffset=true
    Socket Index=&SelectIntFromArray(
     Direct Array Index=&SpawnGroup()
     Direct Array Mode=::NewEnumerator1
     Int Selection Array=DataInterface() {propertyValues {IntData [0,1]}}
    )
    Socket Reader=DataInterface() {propertyValues {
      AttachComponentClass=null
      AttachComponentTag=None
      EditorPreviewAsset=&ThreeDee_Gadget
      SourceActorRef=null
      SourceAsset=null
      SourceMode=Default
      bRequireCurrentFrameData=true
      bUpdateSocketsPerFrame=true
      FilteredSockets [propellorLF1_JNT, propellorRF1_JNT]
      ObjectParameterBinding {parameter {name=None, typeDefHandle {registeredTypeIndex=1}, varData []}}
     }}
    Socket Sampling Mode=::NewEnumerator3
    VelocitySpeedLimit=false
   )
   &AddVelocity()
   /SetVariables(Particles.Position=&AddVectorToPosition(Position=$Output.SocketLocation.SampledPosition, Vector=<0 10 0>)) {isSetParameters=true}
   /SetVariables(Particles.UVDistance:NiagaraFloat=&Multiply_Float(A=&VectorLength(Vector=$Output.SocketLocation.SampledVelocity), B=0.5)) {isSetParameters=true}
  ]
  particleUpdateScript [
   &ParticleState()
   /SetVariables(Particles.RibbonUVDistance:NiagaraFloat=$Particles.UVDistance) {isSetParameters=true}
   &ScaleRibbonWidth(Ribbon Width Scale=&FloatFromCurve(
     FloatCurve=DataInterface() {propertyValues {Curve {keys [
         {arriveTangent=0.1402, interpMode=RCIM_Cubic, leaveTangent=0.1402, tangentMode=RCTM_User, time=0, value=1.7059}
         {arriveTangent=1.32, interpMode=RCIM_Cubic, leaveTangent=1.32, tangentMode=RCTM_User, time=0.5234, value=2.1094}
         {arriveTangent=0.3935, interpMode=RCIM_Cubic, leaveTangent=0.3935, tangentMode=RCTM_User, time=1, value=2.5}
        ]}}}
     Scale Curve=&RandomRangeFloat(Minimum=1, Maximum=1.5)
    )) {module=ScaleRibbonWidth001}
   /Niagara/Modules/Update/Forces/V2/CurlNoiseForce(
    Bias Noise=false
    Noise Frequency=5
    Noise Strength=&FloatFromCurve(
     FloatCurve=DataInterface() {propertyValues {Curve {
        defaultValue=1
        keys [<at 0 0 RCIM_Cubic>, {arriveTangent=1.5, interpMode=RCIM_Cubic, leaveTangent=1.5, time=0.7, value=0.2}, <at 1 1.5 RCIM_Cubic>]
       }}}
     Scale Curve=75
    )
    Pan Noise=false
   )
   &FN_RibbonMaterial(
    Color=&ColorFromCurve(ColorCurve=DataInterface() {propertyValues {
       AlphaCurve {keys [<at 0 2 RCIM_Cubic>,<at 0.0513 2>]}
       BlueCurve {keys [<at 0.0258 0.5089>]}
       GreenCurve {keys [<at 0.0258 0.4793>]}
       RedCurve {keys [<at 0.0258 0.4735>]}
      }})
    Erosion Amount:NiagaraFloat=&Lerp_Float(A=0.25, B=1, Alpha=$Particles.NormalizedAge)
    Erossion Scale U=4.5
    Macro Distortion Amount=0.1
    Macro Pan Offset=-0.1216
    Micro Distortion Amount=0.0384
    Opacity Erosion:NiagaraFloat=&FloatFromCurve(FloatCurve=DataInterface() {propertyValues {Curve {keys [<at 0 0.25 RCIM_Cubic>, {arriveTangent=1.1077, tangentMode=RCTM_User, time=1, value=1}]}}})
    Opacity Power:NiagaraFloat=&FloatFromCurve(FloatCurve=DataInterface() {propertyValues {Curve {keys [<at 0 0.55 RCIM_Cubic>, {arriveTangent=1.1077, tangentMode=RCTM_User, time=1, value=1}]}}})
    Scale Alpha:NiagaraFloat=&FloatFromCurve(
     FloatCurve=DataInterface() {propertyValues {Curve {keys [
         <at 0.084 0 RCIM_Cubic>
         {arriveTangent=1.3088, interpMode=RCIM_Cubic, leaveTangent=1.3088, time=0.1973, value=0.4504}
         {arriveTangent=-0.561, interpMode=RCIM_Cubic, leaveTangent=-0.561, time=0.3048, value=0.289}
         <at 1 0 RCIM_Cubic>
        ]}}}
     Scale Curve=&FloatFromCurve(
      FloatCurve=DataInterface() {propertyValues {Curve {
         defaultValue=1
         keys [
          <at 0.0635 0 RCIM_Cubic>
          {arriveTangentWeight=0.0471, interpMode=RCIM_Cubic, leaveTangentWeight=0.4243, tangentMode=RCTM_User, time=0.4236, value=1}
          <at 1 1 RCIM_Cubic>
         ]
        }}}
      CurveIndex=$Emitter.Age
     )
    )
    UV Basis:NiagaraFloat=&ReturnNormalizedExecIndex()
   ) {module=FN_RibbonMaterial001}
   &GravityForce(Gravity=&Multiply_VectorByFloat(
     Vector=<0 0 1>
     Float=&FloatFromCurve(
      FloatCurve=DataInterface() {propertyValues {Curve {
         defaultValue=1
         keys [<at 0 0 RCIM_Cubic>, {arriveTangent=1.5, interpMode=RCIM_Cubic, leaveTangent=1.5, time=0.5, value=0.6}, <at 1 1.5 RCIM_Cubic>]
        }}}
      Scale Curve=50
     )
    ))
   &SolveForcesAndVelocity(Clamp Velocity=false, Limit Acceleration=false)
  ]
  rendererClasses [&NiagaraRibbonRendererProperties]
  renderers [{rendererClass=&NiagaraRibbonRendererProperties}]
 }
}
summary {
 systemName=NS_Emote_ThreeDee_PropSpawn
 emitters [
  {
   emitterName=Beam
   rendererClasses [
    &NiagaraMeshRendererProperties
    &NiagaraLightRendererProperties
    &NiagaraSpriteRendererProperties
   ]
  }
  {emitterName=LightShaftsBurst, rendererClasses [&NiagaraMeshRendererProperties]}
  {emitterName=Embers_Burst, rendererClasses [&NiagaraSpriteRendererProperties]}
  {emitterName=Main_RingCombined, rendererClasses [&NiagaraMeshRendererProperties]}
  {emitterName=Wind, rendererClasses [&NiagaraRibbonRendererProperties]}
  {emitterName=NE_SpeedLinesCombined001, rendererClasses [&NiagaraMeshRendererProperties]}
 ]
 userVariables [
  User.bIsFrontEnd {type=NiagaraBool}
  User.bIsFrontEndPreview {type=NiagaraBool}
  User.OwnerSkeleton {
   type=NiagaraDataInterfaceSkeletalMesh
   defaultValue {
    struct=NiagaraExt_VariableValue_DataInterface
    value {dataInterface=&self:NiagaraDataInterfaceSkeletalMesh_0}
   }
  }
 ]
}
