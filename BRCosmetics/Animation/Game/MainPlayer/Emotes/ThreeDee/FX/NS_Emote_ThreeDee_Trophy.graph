NiagaraSystem /BRCosmetics/Animation/Game/MainPlayer/Emotes/ThreeDee/FX/NS_Emote_ThreeDee_Trophy
dependencies {
 usedDataInterfaces [
  &NiagaraDataInterfaceSpriteRendererInfo
  &NiagaraDataInterfaceCurve
  &NiagaraDataInterfaceColorCurve
  &NiagaraDataInterfaceArrayColor
 ]
 usedDynamicInputs [
  &FN_LOD_Distance_Remap
  /Niagara/DynamicInputs/Helpers/MaskVector2d
  &MakeCustomVectorFromBool
  &SetBoolByFloatComparison
  &MakeFloatFromVector
  &FNTimeOfDayColor
 ]
 usedModules [
  &self:SystemSpawnScript.SystemScriptSource.SystemScriptGraph.NiagaraNodeAssignment_0.SetVariables_#1
  &SystemState
 ]
}
emitters {
 AirglowLWE {rendererClasses [], renderers []}
 Decal {rendererClasses [], renderers []}
 Embers_Burst {
  emitterUpdateScript [
   &EmitterState(
    Inactive Response=::NewEnumerator1
    Life Cycle Mode=::NewEnumerator1
    Loop Behavior=::NewEnumerator1
    Loop Delay=0.1
    Loop Duration=1
    UseLoopDelay=true
   )
   &SpawnBurst_Instantaneous(Spawn Count=20, Use Spawn Probability=false, UseLoopCountLimit=false)
  ]
  particleSpawnScript [
   &InitializeParticle(
    Color=<rgba 1 1 1 2>
    Lifetime=&UniformRangedFloat(Minimum=0.5, Maximum=0.9)
    Mass Mode=::NewEnumerator1
    Sprite Size Mode=::NewEnumerator1
    Uniform Sprite Size Max=1.5
    Uniform Sprite Size Min=0.8
    UsePositionOffset=false
    Write Material Random=true
   ) {module=InitializeParticle001}
   &ShapeLocation(
    MaskSpawnGroup=false
    Non Uniform Scale=<1 1 1.5>
    Offset=<0 0 10>
    Offset Coordinate Space=::World
    Offset Mode=::NewEnumerator0
    Override Seed=false
    Sphere Radius=25
    UseOwnerScale=true
   )
   &AddVelocity(Velocity=&RandomRangeVector(Minimum=<0 0 25>, Maximum=<0 0 0>)) {module=AddVelocity001}
   &CameraOffset(Camera Offset Amount=-10)
  ]
  particleUpdateScript [
   &ParticleState()
   &ApplyOwnerScaleToAttributes(Scale Initial Velocity=true)
   &ScaleSpriteSize(
    Scale Sprite Size Mode=::NewEnumerator0
    Uniform Scale Factor=&Lerp_Float(A=1.5, B=2, Alpha=$System.DistanceScalar)
   )
   &Drag(
    Drag=&Multiply_Float(
     A=&FloatFromCurve(
      FloatCurve=DataInterface() {propertyValues {Curve {keys [<at 0 0.5>, {arriveTangent=0.5, interpMode=RCIM_Cubic, leaveTangent=0.5, time=0.3, value=0.5}, <at 1 1 RCIM_Cubic>]}}}
      Scale Curve=12
     )
     B=1
    )
    Use Linear Drag=true
    Use Rotational Drag=false
   )
   /Niagara/Modules/Update/Forces/CurlNoiseForce(
    Noise Frequency=10
    Noise Strength=&FloatFromCurve(FloatCurve=DataInterface() {propertyValues {Curve {keys [<at 0 1>,<at 1 0>]}}}, Scale Curve=2000)
    Pan Noise=true
    Pan Noise Field=<0.1 0.1 0.5>
   )
   &GravityForce(Gravity=&RandomRangeVector(Minimum=<0 0 900>, Maximum=<0 0 -500>))
   &InheritSourceMovement(
    Applied Position Scale=&MakeVector(
     X=0
     Y=0
     Z=&FloatFromCurve(
      FloatCurve=DataInterface() {propertyValues {Curve {
         defaultValue=1
         keys [
          {interpMode=RCIM_Cubic, tangentMode=RCTM_User, time=0, value=0.0874}
          {arriveTangent=3.8536, interpMode=RCIM_Cubic, leaveTangent=3.8536, tangentMode=RCTM_User, time=0.1, value=0.5931}
         ]
        }}}
      Scale Curve=-1
     )
    )
    ApplyForce=false
    ApplyPosition=true
    ApplyPositionFalloff=false
   )
   &SolveForcesAndVelocity(Clamp Velocity=false, Limit Acceleration=false)
   &ScaleSpriteSizeBySpeed(
    Max Scale Factor=<0.8 5>
    Min Scale Factor=<0.8 0.8>
    Sample Scale Factor By Curve=true
    Scale Factor Curve=DataInterface() {propertyValues {Curve {keys [
        <at -0.0077 0.0765 RCIM_Cubic>
        {arriveTangent=1.2463, interpMode=RCIM_Cubic, leaveTangent=1.2463, tangentMode=RCTM_User, time=0.9985, value=0.9549}
       ]}}}
    Velocity Threshold=300
   )
   &ScaleColor(
    Scale Alpha=&FloatFromCurve(
     FloatCurve=DataInterface() {propertyValues {Curve {keys [<at 0.25 1>,<at 1 0>]}}}
     Scale Curve=&FNTimeOfDayScalar(Night Scale=0.1, Morning Scale=0.2, Evening Scale=0.2)
    )
    Scale RGB=&MakeVectorFromLinearColorRGB(Color=&ColorFromCurve(ColorCurve=DataInterface() {propertyValues {
        AlphaCurve {keys [<at 0 1 RCIM_Cubic>,<at 0.7107 1>]}
        BlueCurve {keys [<at 0 14.6701 RCIM_Cubic>,<at 0.368 3.776>,<at 0.855 0.1753 RCIM_Cubic>]}
        GreenCurve {keys [<at 0 36.7481 RCIM_Cubic>,<at 0.368 13.2325>,<at 0.855 0.3243 RCIM_Cubic>]}
        RedCurve {keys [<at 0 50 RCIM_Cubic>,<at 0.368 25>,<at 0.855 1 RCIM_Cubic>]}
       }}))
    ScaleA=true
    ScaleRGB=true
   )
  ]
  rendererClasses [&NiagaraSpriteRendererProperties]
  renderers [{rendererClass=&NiagaraSpriteRendererProperties}]
 }
 Flash {
  emitterUpdateScript [
   &EmitterState(Life Cycle Mode=::NewEnumerator1, Loop Behavior=::NewEnumerator1, Loop Duration=1, UseLoopDelay=false)
   &SpawnBurst_Instantaneous(Use Spawn Probability=false, UseLoopCountLimit=false)
   &SpawnBurst_Instantaneous(Spawn Group=1, Use Spawn Probability=false, UseLoopCountLimit=false) {module=SpawnBurst_Instantaneous001}
  ]
  particleSpawnScript [
   /SetVariables(Particles.VisibilityTag=&SpawnGroup()) {isSetParameters=true}
   &InitializeParticle(
    Color=&MaskLinearColorBySpawnGroup(Masked Linear Color=<rgba 2 1.0841 0 1>, Passthrough Linear Color=<rgba 1 0.6089 0.0099 1>)
    Lifetime=&MaskFloatBySpawnGroup(Masked Float=0.6, Passthrough Float=0.15)
    Position Offset=&MaskVectorBySpawnGroup(Masked Vector=<0 0 20>, Passthrough Vector=<0 0 15>)
    Sprite Size Mode=::NewEnumerator3
    Uniform Sprite Size=&MaskFloatBySpawnGroup(Masked Float=150, Passthrough Float=10)
    UsePositionOffset=true
    Write Material Random=true
   )
   &CameraOffset(Camera Offset Amount=&MaskFloatBySpawnGroup(Masked Float=-20, Passthrough Float=20))
  ]
  particleUpdateScript [
   &SolveForcesAndVelocity(Clamp Velocity=false, Limit Acceleration=false)
   &ParticleState()
   &SpriteRotationRate(Rotation Rate=&MaskFloatBySpawnGroup(
     Masked Float=&FloatFromCurve(
      FloatCurve=DataInterface() {propertyValues {Curve {
         defaultValue=0
         keys [
          {arriveTangent=4.1785, interpMode=RCIM_Cubic, leaveTangent=4.1785, tangentMode=RCTM_User, time=0.2036, value=0}
          {arriveTangent=1.1179, interpMode=RCIM_Cubic, leaveTangent=1.1179, tangentMode=RCTM_User, time=0.4, value=0.725}
          {arriveTangent=0.0224, interpMode=RCIM_Cubic, leaveTangent=0.0224, tangentMode=RCTM_User, time=1, value=1}
         ]
        }}}
      Scale Curve=100
     )
     Passthrough Float=&FloatFromCurve(
      FloatCurve=DataInterface() {propertyValues {Curve {keys [
          {interpMode=RCIM_Cubic, tangentMode=RCTM_User, time=0, value=1}
          {arriveTangent=-1.2549, interpMode=RCIM_Cubic, leaveTangent=-1.2549, tangentMode=RCTM_User, time=1, value=0}
         ]}}}
      Scale Curve=20
     )
    ))
   &ScaleSpriteSize(Uniform Curve Sprite Scale=DataInterface() {propertyValues {Curve {
       defaultValue=0.3921
       keys [
        {interpMode=RCIM_Cubic, tangentMode=RCTM_User, time=0, value=0.6142}
        {interpMode=RCIM_Cubic, tangentMode=RCTM_User, time=0.1127, value=1}
        {arriveTangent=-0.0416, interpMode=RCIM_Cubic, leaveTangent=-0.0416, time=0.5229, value=0.567}
        {arriveTangent=-1.1885, interpMode=RCIM_Cubic, leaveTangent=-1.1885, time=0.885, value=0.9679}
        {tangentMode=RCTM_User, time=1, value=0}
       ]
      }}})
   &DynamicMaterialParameters(
    Index 0 Param 1=1
    Index 0 Param 3=&FloatFromCurve(
     FloatCurve=DataInterface() {propertyValues {Curve {keys [{interpMode=RCIM_Cubic, leaveTangent=8.1085, tangentMode=RCTM_Break, time=0.8568, value=1}, {tangentMode=RCTM_User, time=1, value=0}]}}}
     Scale Curve=0.015
    )
    Index 0 Param 4=0.8
    Index 1 Param 1=&FloatFromCurve(FloatCurve=DataInterface() {propertyValues {Curve {keys [
         {interpMode=RCIM_Cubic, leaveTangent=1.1034, tangentMode=RCTM_Break, time=0, value=0}
         {arriveTangent=2.2399, tangentMode=RCTM_User, time=0.5, value=1}
        ]}}})
    Index 1 Param 2=&FloatFromCurve(FloatCurve=DataInterface() {propertyValues {Curve {keys [<at 0.5 0>, {arriveTangent=3.2053, tangentMode=RCTM_User, time=0.9, value=1}]}}})
    Index 1 Param 3=&FloatFromCurve(FloatCurve=DataInterface() {propertyValues {Curve {keys [<at 0 0 RCIM_Cubic>, {arriveTangent=3.4223, interpMode=RCIM_Cubic, leaveTangent=3.4223, tangentMode=RCTM_User, time=1, value=1}]}}})
    Index 1 Param 4=&FloatFromCurve(
     FloatCurve=DataInterface() {propertyValues {Curve {keys [
         {interpMode=RCIM_Cubic, leaveTangent=0.1169, tangentMode=RCTM_Break, time=0, value=0}
         {arriveTangent=2.5708, interpMode=RCIM_Cubic, leaveTangent=2.5708, tangentMode=RCTM_User, time=1, value=1}
        ]}}}
     Scale Curve=&RandomRangeFloat(Minimum=1, Maximum=40)
    )
    Index 2 Param 1=&FloatFromCurve(FloatCurve=DataInterface() {propertyValues {Curve {keys [
         {interpMode=RCIM_Cubic, leaveTangent=-0.0025, tangentMode=RCTM_Break, time=0, value=0.5}
         {arriveTangent=-0.0316, interpMode=RCIM_Cubic, tangentMode=RCTM_Break, time=1, value=-0.2}
        ]}}})
    Index 2 Param 2=&FloatFromCurve(
     FloatCurve=DataInterface() {propertyValues {Curve {keys [
         {interpMode=RCIM_Cubic, tangentMode=RCTM_User, time=0, value=0}
         {arriveTangent=2.0522, interpMode=RCIM_Cubic, leaveTangent=2.0522, tangentMode=RCTM_User, time=1, value=1}
        ]}}}
     Scale Curve=2
    )
    Module.Index0Float0Write=true
    Module.Index0Float1Write=true
    Module.Index0Float2Write=true
    Module.Index0Float3Write=true
    Module.Index1Float0Write=true
    Module.Index1Float1Write=true
    Module.Index1Float2Write=true
    Module.Index1Float3Write=true
    Module.Index2Float0Write=true
    Module.Index2Float1Write=true
    Module.Index2Float2Write=false
    Module.Index2Float3Write=false
    Module.Index3Float0Write=false
    Module.Index3Float1Write=false
    Module.Index3Float2Write=false
    Module.Index3Float3Write=true
    Write Parameter Index 1=::NewEnumerator1
    Write Parameter Index 2=::NewEnumerator1
    Write Parameter Index 3=::NewEnumerator1
   )
   &ScaleColor(
    Scale Alpha=&MaskFloatBySpawnGroup(
     Masked Float=&FloatFromCurve(
      FloatCurve=DataInterface() {propertyValues {Curve {keys [
          {tangentMode=RCTM_User, time=0, value=1}
          {arriveTangent=-2.2488, interpMode=RCIM_Cubic, leaveTangent=-2.2488, tangentMode=RCTM_User, time=0.2763, value=1}
          {arriveTangent=-0.9515, interpMode=RCIM_Cubic, leaveTangent=-0.9515, tangentMode=RCTM_User, time=1, value=0}
         ]}}}
      Scale Curve=&Multiply_Float(
       A=&FN_FrontEndFloatModifier(Gameplay=1, Front End=0.5)
       B=&MaskFloatBySpawnGroup(
        Masked Float=1
        Passthrough Float=&FloatFromCurve(
         FloatCurve=DataInterface() {propertyValues {Curve {
            defaultValue=1
            keys [
             {arriveTangent=8.9136, interpMode=RCIM_Cubic, leaveTangent=8.9136, tangentMode=RCTM_User, time=0, value=0}
             {
              arriveTangentWeight=0.1126
              interpMode=RCIM_Cubic
              leaveTangentWeight=0.1126
              tangentMode=RCTM_User
              tangentWeightMode=RCTWM_WeightedBoth
              time=0.3264
              value=1
             }
             {arriveTangent=-4.0894, interpMode=RCIM_Cubic, leaveTangent=-4.0894, tangentMode=RCTM_User, time=1.0009, value=0}
            ]
           }}}
         Scale Curve=&FNTimeOfDayScalar(Night Scale=0.5, Morning Scale=1, Evening Scale=1)
        )
       )
      )
     )
     Passthrough Float=&FloatFromCurve(FloatCurve=DataInterface() {propertyValues {Curve {
         defaultValue=1
         keys [
          {interpMode=RCIM_Cubic, tangentMode=RCTM_User, time=0, value=1}
          {arriveTangent=0.1313, interpMode=RCIM_Cubic, leaveTangent=0.1313, time=0.0904, value=1.3675}
          {arriveTangent=-0.0904, interpMode=RCIM_Cubic, leaveTangent=-0.0904, tangentMode=RCTM_User, time=0.1206, value=1.0158}
          {arriveTangent=0.4479, interpMode=RCIM_Cubic, leaveTangent=0.4479, time=0.8158, value=1.0409}
          {arriveTangent=-5.6527, interpMode=RCIM_Cubic, leaveTangent=-5.6527, time=0.872, value=1.3524}
          {arriveTangent=-2.9571, interpMode=RCIM_Cubic, leaveTangent=-2.9571, tangentMode=RCTM_User, time=1, value=0}
         ]
        }}})
    )
    Scale RGB=&Lerp_Vector(
     A=&VectorFromFloat(Value=&FloatFromCurve(
       FloatCurve=DataInterface() {propertyValues {Curve {keys [<at 0 1>,<at 0.1 1>,<at 0.2 2>,<at 0.3 1>,<at 0.6 1>,<at 0.7 2>,<at 0.8 1>,<at 1 1>]}}}
       Scale Curve=10
      ))
     B=&VectorFromFloat(Value=&RandomRangeFloat(Minimum=0.3, Maximum=5, Evaluation Type=::NewEnumerator1))
     Alpha=&FloatFromCurve(FloatCurve=DataInterface() {propertyValues {Curve {keys [
          {interpMode=RCIM_Cubic, leaveTangent=2.5294, tangentMode=RCTM_Break, time=0, value=0}
          {arriveTangent=-0.2166, interpMode=RCIM_Cubic, tangentMode=RCTM_Break, time=0.4, value=1}
         ]}}})
    )
    ScaleA=true
    ScaleRGB=false
   )
   /SetVariables(
    Particles.FlareDynPar:Vector4f=<1 0 0 0>
    Particles.FlareDynPar2:Vector4f=<2.5 0 1 1>
    Particles.FlareDynPar3:Vector4f=<30 0 0 0>
    Particles.FlareSpriteSize:Vector2f=&Vector2DFromFloat(Value=&FloatFromCurve(
      FloatCurve=DataInterface() {propertyValues {Curve {keys [
          {interpMode=RCIM_Cubic, leaveTangent=5.6981, tangentMode=RCTM_Break, time=0, value=0.2}
          <at 0.315 1>
          {arriveTangent=-0.209, interpMode=RCIM_Cubic, leaveTangent=-0.209, time=4.9, value=1}
          {arriveTangent=-13.1443, interpMode=RCIM_Cubic, leaveTangent=-13.1443, tangentMode=RCTM_User, time=5.1, value=0}
         ]}}}
      CurveIndex=$Particles.Age
      Scale Curve=300
     ))
   ) {isSetParameters=true}
  ]
  rendererClasses [&NiagaraSpriteRendererProperties]
  renderers [
   {rendererClass=&NiagaraSpriteRendererProperties}
   {rendererClass=&NiagaraSpriteRendererProperties, rendererIndex=1}
  ]
 }
 Light_Glints {
  emitterUpdateScript [
   &EmitterState(Life Cycle Mode=::NewEnumerator1, Loop Behavior=::NewEnumerator1, Loop Duration=1.5, UseLoopDelay=false)
   &SpawnBurst_Instantaneous(
    Spawn Count=&RandomRangeInt(Minimum=3, Maximum=5)
    Use Spawn Probability=false
    UseLoopCountLimit=false
   )
  ]
  particleSpawnScript [
   &InitializeParticle(
    Color Mode=::NewEnumerator0
    Lifetime Max=1
    Lifetime Min=1.5
    Lifetime Mode=::NewEnumerator1
    Sprite Rotation Mode=::NewEnumerator2
    Sprite Size Mode=::NewEnumerator1
    Uniform Sprite Size Max=30
    Uniform Sprite Size Min=10
    UsePositionOffset=false
    Write Material Random=true
   )
   &ShapeLocation(
    MaskSpawnGroup=false
    Non Uniform Scale=<1 1 3>
    Offset=<0 0 30>
    Offset Mode=::NewEnumerator0
    Override Seed=false
    Sphere Radius=20
    UseOwnerScale=true
   )
   &AddVelocity(
    Module.Use Falloff Distance=false
    Override Seed=false
    Velocity Mode=::NewEnumerator1
    Velocity Origin=$Particles.ShapeLocation.ShapePosition
    Velocity Speed=&RandomRangeFloat(Minimum=100, Maximum=50)
   )
  ]
  particleUpdateScript [
   &ParticleState()
   &ScaleSpriteSize(
    Scale Sprite Size Mode=::NewEnumerator0
    Uniform Scale Factor=&MaskFloat(
     Mask Source=0
     Mask Parameter=?
     Mask=0
     Masked Float=&FloatFromCurve(FloatCurve=DataInterface() {propertyValues {Curve {keys [
          {arriveTangent=7.8653, interpMode=RCIM_Cubic, leaveTangent=7.8653, tangentMode=RCTM_User, time=0.0033, value=0.0151}
          {arriveTangent=0.8866, interpMode=RCIM_Cubic, leaveTangent=0.8866, tangentMode=RCTM_User, time=0.2773, value=0.783}
          {arriveTangent=12.6874, interpMode=RCIM_Cubic, leaveTangent=-13.0744, tangentMode=RCTM_Break, time=0.4458, value=1.3521}
          {arriveTangent=-0.8088, interpMode=RCIM_Cubic, leaveTangent=-0.8088, tangentMode=RCTM_User, time=0.6094, value=0.8056}
          {arriveTangent=-5.9884, interpMode=RCIM_Cubic, leaveTangent=-5.9884, tangentMode=RCTM_User, time=1, value=0}
         ]}}})
     Passthrough Float=&FloatFromCurve(FloatCurve=DataInterface() {propertyValues {Curve {keys [
          <at 0 0 RCIM_Cubic>
          {arriveTangent=3.1939, interpMode=RCIM_Cubic, leaveTangent=2.7137, tangentMode=RCTM_Break, time=0.45, value=1}
          {arriveTangent=-1.7556, interpMode=RCIM_Cubic, leaveTangent=-2.8595, tangentMode=RCTM_Break, time=0.55, value=1}
          {arriveTangent=0.4826, interpMode=RCIM_Cubic, leaveTangent=0.4826, tangentMode=RCTM_User, time=1, value=0}
         ]}}})
    )
   )
   /SetVariables(Particles.Glitter Color:LinearColor=&Multiply_LinearColorByFloat(
     Linear Color=&SelectLinearColorFromArray(
      Array Sampling Mode=::NewEnumerator0
      Direct Array Index=?
      Color Selection Array=DataInterface() {propertyValues {ColorData [<rgba 1 0.8958 0.28 0.2>, <rgba 1 0.2505 0.0561 0.2>, <rgba 1 0.3848 0.1346 0.2>, <rgba 1 0.6061 0.0764 0.2>, <rgba 1 0.7723 0.1541 0.2>]}}
      Random Seed=0
     )
     Float=&FloatFromCurve(
      FloatCurve=DataInterface() {propertyValues {Curve {keys [
          {arriveTangent=-0.012, interpMode=RCIM_Cubic, leaveTangent=-0.012, tangentMode=RCTM_User, time=0, value=0}
          {arriveTangent=1.1429, interpMode=RCIM_Cubic, leaveTangent=1.1429, time=0.2, value=1}
          {interpMode=RCIM_Cubic, tangentMode=RCTM_User, time=0.35, value=0.4}
          {interpMode=RCIM_Cubic, tangentMode=RCTM_User, time=0.5, value=0.6}
          {arriveTangent=-0.9375, interpMode=RCIM_Cubic, leaveTangent=-0.9375, time=0.68, value=0.15}
          {arriveTangent=-0.4688, interpMode=RCIM_Cubic, leaveTangent=-0.4688, time=0.82, value=0.3}
          {interpMode=RCIM_Cubic, tangentMode=RCTM_Break, time=1, value=0}
          {interpMode=RCIM_Cubic, leaveTangent=-1.75, time=5, value=0.2}
         ]}}}
      Scale Curve=&Multiply_Float(A=0.85, B=&FNTimeOfDayScalar(Night Scale=0.7, Morning Scale=0.8, Evening Scale=0.8))
     )
    )) {isSetParameters=true}
   /Niagara/Modules/Update/Forces/V2/CurlNoiseForce(
    Bias Noise=false
    Noise Frequency=20
    Noise Strength=&FloatFromCurve(FloatCurve=DataInterface() {propertyValues {Curve {keys [{arriveTangent=-158.0912, interpMode=RCIM_Cubic, leaveTangent=-158.0912, tangentMode=RCTM_User, time=0, value=50}, <at 1 0>]}}})
    Pan Noise=false
   )
   &Drag(
    Drag=&FloatFromCurve(
     FloatCurve=DataInterface() {propertyValues {Curve {
        defaultValue=1
        keys [{interpMode=RCIM_Cubic, tangentMode=RCTM_User, time=0, value=0}, {interpMode=RCIM_Cubic, tangentMode=RCTM_User, time=0.3228, value=1}]
       }}}
     Scale Curve=2
    )
    Use Linear Drag=true
    Use Rotational Drag=true
   )
   &InheritSourceMovement(
    Applied Position Scale=&MakeVector(
     X=0
     Y=0
     Z=&FloatFromCurve(
      FloatCurve=DataInterface() {propertyValues {Curve {
         defaultValue=1
         keys [
          {interpMode=RCIM_Cubic, tangentMode=RCTM_User, time=0, value=0.2155}
          {arriveTangent=3.8536, interpMode=RCIM_Cubic, leaveTangent=3.8536, tangentMode=RCTM_User, time=0.272, value=0.5931}
         ]
        }}}
      Scale Curve=-1
     )
    )
    ApplyForce=false
    ApplyPosition=true
    ApplyPositionFalloff=false
   )
   &SolveForcesAndVelocity(Clamp Velocity=false, Limit Acceleration=false)
  ]
  rendererClasses [&NiagaraSpriteRendererProperties]
  renderers [{rendererClass=&NiagaraSpriteRendererProperties}]
 }
 Smoke {
  emitterUpdateScript [
   &EmitterState(Life Cycle Mode=::NewEnumerator1, Loop Behavior=::NewEnumerator1, Loop Duration=1.5, UseLoopDelay=false)
   &SpawnBurst_Instantaneous(Spawn Count=3, Use Spawn Probability=false, UseLoopCountLimit=false)
  ]
  particleSpawnScript [
   &InitializeParticle(
    Lifetime Max=0.8
    Lifetime Min=1.2
    Lifetime Mode=::NewEnumerator1
    Sprite Rotation Mode=::NewEnumerator2
    Sprite Size Mode=::NewEnumerator1
    Uniform Sprite Size Max=42
    Uniform Sprite Size Min=32
    UsePositionOffset=false
    Write Material Random=true
   )
   &ShapeLocation(
    MaskSpawnGroup=false
    Non Uniform Scale=<0.2 0.2 7>
    Offset=<0 0 -5>
    Offset Mode=::NewEnumerator0
    Override Seed=false
    Sphere Radius=5
    UseOwnerScale=true
   )
   &AddVelocity(
    Cone Axis=<0 0 1>
    Module.UseConeFalloff=false
    Module.UseVelDistribution=false
    Override Seed=false
    Velocity Mode=::NewEnumerator2
    Velocity Speed=&RandomRangeFloat(Minimum=20, Maximum=0)
   )
   &AddVelocity(
    Module.Use Falloff Distance=false
    Override Seed=false
    Velocity Mode=::NewEnumerator1
    Velocity Origin=$Particles.ShapeLocation.ShapePosition
    Velocity Speed=&RandomRangeFloat(Minimum=0, Maximum=25)
   ) {module=AddVelocity001}
  ]
  particleUpdateScript [
   &ParticleState()
   &SubUVAnimation(
    "Sprite Renderer "=DataInterface() {propertyValues {SpriteRenderer=&self:Smoke_1.NiagaraSpriteRendererProperties_0}}
    UseEndFrame=false
    UseStartFrame=false
   )
   &ScaleSpriteSize(
    Uniform Curve Scale=1.45
    Uniform Curve Sprite Scale=DataInterface() {propertyValues {Curve {
       defaultValue=0
       keys [
        {arriveTangent=4.1785, interpMode=RCIM_Cubic, leaveTangent=4.1785, tangentMode=RCTM_User, time=0, value=0}
        {arriveTangent=1.1179, interpMode=RCIM_Cubic, leaveTangent=1.1179, tangentMode=RCTM_User, time=0.2001, value=0.7483}
        {arriveTangent=0.0224, interpMode=RCIM_Cubic, leaveTangent=0.0224, tangentMode=RCTM_User, time=1, value=1}
       ]
      }}}
   )
   &ScaleColor(
    Scale Alpha=&FloatFromCurve(
     FloatCurve=DataInterface() {propertyValues {Curve {
        defaultValue=0
        keys [
         {arriveTangent=9.0725, interpMode=RCIM_Cubic, leaveTangent=9.0725, tangentMode=RCTM_User, time=0, value=0.7366}
         {arriveTangent=-2.5249, interpMode=RCIM_Cubic, leaveTangent=-2.5249, tangentMode=RCTM_User, time=0.2536, value=1.2737}
         {arriveTangent=-0.0405, interpMode=RCIM_Cubic, leaveTangent=-0.0405, tangentMode=RCTM_User, time=1, value=0}
        ]
       }}}
     Scale Curve=0.45
    )
    Scale RGB=&MakeVectorFromLinearColorRGB(Color=&ColorFromCurve(ColorCurve=DataInterface() {propertyValues {
        AlphaCurve {keys []}
        BlueCurve {keys [<at 0 0.2129 RCIM_Cubic>,<at 0.0393 0.101>,<at 1 0 RCIM_Cubic>]}
        GreenCurve {keys [
          <at 0 3.3791 RCIM_Cubic>
          <at 0.0393 0.937>
          {arriveTangent=-1.2172, interpMode=RCIM_Cubic, leaveTangent=-1.2172, tangentMode=RCTM_User, time=1, value=0.3105}
         ]}
        RedCurve {keys [
          <at 0 5 RCIM_Cubic>
          <at 0.0393 2>
          {arriveTangent=-1.6259, interpMode=RCIM_Cubic, leaveTangent=-1.6259, tangentMode=RCTM_User, time=1, value=1}
         ]}
       }}))
    ScaleA=true
    ScaleRGB=true
   )
   &Drag(Drag=&RandomRangeFloat(Minimum=0.8, Maximum=1.2), Use Linear Drag=true, Use Rotational Drag=true)
   &InheritSourceMovement(
    Applied Position Scale=&MakeVector(
     X=0
     Y=0
     Z=&FloatFromCurve(
      FloatCurve=DataInterface() {propertyValues {Curve {
         defaultValue=1
         keys [
          {interpMode=RCIM_Cubic, tangentMode=RCTM_User, time=0, value=0.0874}
          {arriveTangent=3.8536, interpMode=RCIM_Cubic, leaveTangent=3.8536, tangentMode=RCTM_User, time=0.0539, value=0.5931}
          {arriveTangent=1.0538, interpMode=RCIM_Cubic, leaveTangent=1.0538, time=0.5978, value=0.59}
          {interpMode=RCIM_Cubic, leaveTangent=0.547, time=0.6298, value=1.2}
         ]
        }}}
      Scale Curve=-1
     )
    )
    ApplyForce=false
    ApplyPosition=true
    ApplyPositionFalloff=false
   )
   &SolveForcesAndVelocity(Clamp Velocity=false, Limit Acceleration=false)
   &DynamicMaterialParameters(
    Index 0 Param 1=3
    Index 0 Param 2=1
    Index 0 Param 3=1
    Module.Index0Float0Write=true
    Module.Index0Float1Write=true
    Module.Index0Float2Write=true
    Module.Index0Float3Write=true
   )
   &CameraOffset(Camera Offset Amount=&FloatFromCurve(
     FloatCurve=DataInterface() {propertyValues {Curve {
        defaultValue=1
        keys [{interpMode=RCIM_Cubic, tangentMode=RCTM_User, time=0, value=0.5}, {interpMode=RCIM_Cubic, tangentMode=RCTM_User, time=0.3858, value=-1}]
       }}}
     Scale Curve=&RandomRangeFloat(Minimum=10, Maximum=20)
    ))
  ]
  rendererClasses [&NiagaraSpriteRendererProperties]
  renderers [{rendererClass=&NiagaraSpriteRendererProperties}]
 }
 Sparkles {rendererClasses [], renderers []}
 TrophyGlow {rendererClasses [], renderers []}
}
summary {
 systemName=NS_Emote_ThreeDee_Trophy
 emitters [
  {emitterName=Smoke, rendererClasses [&NiagaraSpriteRendererProperties]}
  {emitterName=Embers_Burst, rendererClasses [&NiagaraSpriteRendererProperties]}
  {emitterName=AirglowLWE, rendererClasses []}
  {emitterName=Light_Glints, rendererClasses [&NiagaraSpriteRendererProperties]}
  {emitterName=Flash, rendererClasses [&NiagaraSpriteRendererProperties]}
  {emitterName=Sparkles, rendererClasses []}
  {emitterName=Decal, rendererClasses []}
  {emitterName=TrophyGlow, rendererClasses []}
 ]
 userVariables [
  User.bIsFrontEnd {type=NiagaraBool}
  User.bIsFrontEndPreview {type=NiagaraBool}
  User.LightningColor {type=LinearColor, description="", defaultValue {struct=LinearColor, value=<rgba 1 1 1 1>}}
 ]
}
